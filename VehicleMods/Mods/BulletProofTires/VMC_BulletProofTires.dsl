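; Vehicle mod component for bulletproof tires on Fortnite Athena vehicles.
; On BeginPlay it casts its owning actor to FortAthenaVehicle and stores the result as the owner vehicle.
; The mod enabled and disabled events each check that stored vehicle, but both branches are empty, so this graph applies no tire behavior itself.
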(event EventBeginPlay
  (bind _asfort_athena_vehicle (Utilities|Casting|CastToFortAthenaVehicle (Components|GetOwner)))
  (Variables|Default|SetOwnerVehicle _asfort_athena_vehicle))

(event VehicleMods|EventOnModEnabled
  (Variables|Default|GetOwnerVehicle
    (:then)
    (:else)))

(event VehicleMods|EventOnModDisabled
  (Variables|Default|GetOwnerVehicle
    (:then)
    (:else)))
